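# Unlit prototype material for one hydraulic erosion pass over the RT texture, which stores height in red and water in blue.
# Custom HLSL adds rain and removes soluble sediment in a cross neighbourhood, then pools summed water from the lowest cells upward.
# Evaporation redeposits sediment at the centre texel, returning new height and water.
# Two variants exist: five-tap driven by MPC_Erosion, and nine-tap driven by local scalar parameters.
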Material /Landmass/Landscape/OldPrototype_BP/Mats/Old/M_ErosionPass_01
graph {
 properties {bAllowNegativeEmissiveColor=true, pixelDepthOffsetMode=PDOM_Legacy, shadingModel=MSM_Unlit, usedShadingModels=Unlit}
 nodes [
  MaterialExpressionTextureObjectParameter_0 = TextureObjectParameter() {parameterName=RT, texture=&Black}
  MaterialExpressionAppendVector_0 = AppendVector(
   A=CollectionParameter() {collection=/Landmass/Landscape/BlueprintBrushes/MPC/MPC_Erosion, parameterName=ResX}
   B=CollectionParameter() {collection=&MPC_Erosion, parameterName=ResY}
  )
  MaterialExpressionCollectionParameter_2 = CollectionParameter() {collection=&MPC_Erosion, parameterName=RainAmount}
  MaterialExpressionCollectionParameter_3 = CollectionParameter() {collection=&MPC_Erosion, parameterName=Solubility}
  MaterialExpressionCollectionParameter_4 = CollectionParameter() {collection=&MPC_Erosion, parameterName=Evaporation}
  MaterialExpressionTextureCoordinate_5 = TextureCoordinate()
  Custom!(
   HeightTex=@MaterialExpressionTextureObjectParameter_0
   TextureSize=@MaterialExpressionAppendVector_0
   RainAmount=@MaterialExpressionCollectionParameter_2
   Solubility=@MaterialExpressionCollectionParameter_3
   Evaporation=@MaterialExpressionCollectionParameter_4
   UV=@MaterialExpressionTextureCoordinate_5
  ) {
   code="//HeightTex: input Texture Object containing heights.R and water.B\r\n//TextureSize: input for size\r\n//RainAmount: input for per texel rain contribution\r\n//Solubility: how much sediment the water can hold\r\n//Evaporation: how much water to evaporate each round\r\n\r\nfloat2 offsets[5] = {float2(0,1),float2(0,-1), float2(0,0),float2(1,0),float2(-1,0)};\r\n\r\nfloat heights[5] = {0,0,0,0,0};\r\nfloat water[5] =   {0,0,0,0,0};\r\nint orders[5] =    {0,0,0,0,0};\r\nint actives[5] =   {0,0,0,0,0};\r\n\r\nfloat2 texelsize = 1 / TextureSize;\r\nfloat watersum = 0;\r\n\r\nUV = (floor(UV * TextureSize)/TextureSize) + (texelsize * 0.5);\r\n\r\n//Sample Heightmap and water data\r\nfor (int i = 0; i < 5; i++)\r\n{\r\n\tfloat2 heightsample = HeightTex.SampleLevel(HeightTexSampler, UV + offsets[i] * texelsize, 0).rb;\r\n\theights[i] = heightsample.r;\r\n\tfloat watertemp = heightsample.g + RainAmount;\r\n\theights[i] -= RainAmount * Solubility; \r\n\r\n\twatersum += watertemp;\r\n\twater[i] = 0;\r\n}\r\n\r\n//sort heights by how many are lower than each\r\nfor (int j = 0; j < 5; j++)\r\n{\r\n\tfloat curheight = heights[j];\r\n\r\n\tfor (int k = 0; k < 5; k++)\r\n\t{\r\n\t\tfloat checkheight = heights[k];\r\n\t\tif (checkheight < curheight)\r\n\t\t{\r\n\t\t\torders[j] += 1;\r\n\t\t}\r\n\t}\r\n}\r\n\r\nfloat maxlevel = 5;\r\n\r\n//use sort data to iterate again\r\nfor (int l = 0; l < min(maxlevel,5); l++)\r\n{\r\n\r\n\tfloat curactivecells = 0;\r\n\tfloat curpooltop = 65000;\r\n\tfloat curpoolbottom = -65000;\r\n\t\r\n\tfor (int m = 0; m < 5; m++)\r\n\t{\r\n\t\tif(orders[m] <= l)\r\n\t\t{\r\n\t\t\tcuractivecells += 1;\r\n\t\t\tcurpoolbottom = max(curpoolbottom,heights[m]+water[m]);\r\n\t\t\tactives[m] = 1;\r\n\t\t}\r\n\t\telse\r\n\t\t{\r\n\t\t\tactives[m] = 0;\r\n\t\t}\r\n\t\tif(orders[m] > l )\r\n\t\t{\r\n\t\t\tcurpooltop = min(curpooltop,heights[m]);\r\n\t\t}\r\n\t\t\r\n\t}\r\n\r\n\tfloat curpoolsize = (curpooltop - curpoolbottom) * curactivecells;\r\n\tcurpoolsize = min(curpoolsize, watersum);\r\n\tcurpoolsize = max(0, curpoolsize);\r\n\r\n\tfor (int n = 0; n < 5; n++)\r\n\t{\r\n\t\tif(actives[n] == 1)\r\n\t\t{\r\n\t\t\twater[n] += curpoolsize / curactivecells;\r\n\t\t\twatersum -= curpoolsize / curactivecells;\r\n\t\t}\r\n\t}\r\n\t\r\n\r\n}\r\n\r\nheights[2] += water[2] * Evaporation * Solubility;\r\nwater[2] -= water[2] * Evaporation;\r\n\r\nreturn float3(heights[2],0,water[2]);"
   description=Erosion Pass
   inputs [
    {
     inputName=HeightTex
     input {expression=&self:MaterialExpressionTextureObjectParameter_0}
    }
    {inputName=TextureSize, input {expression=&self:MaterialExpressionAppendVector_0}}
    {
     inputName=RainAmount
     input {expression=&self:MaterialExpressionCollectionParameter_2}
    }
    {
     inputName=Solubility
     input {expression=&self:MaterialExpressionCollectionParameter_3}
    }
    {
     inputName=Evaporation
     input {expression=&self:MaterialExpressionCollectionParameter_4}
    }
    {inputName=UV, input {expression=&self:MaterialExpressionTextureCoordinate_5}}
   ]
  }
  MaterialExpressionTextureObjectParameter_1 = TextureObjectParameter() {parameterName=RT, texture=&Black}
  MaterialExpressionAppendVector_1 = AppendVector(A=ScalarParameter() {defaultValue=1024, parameterName=SizeX}, B=ScalarParameter() {defaultValue=1024, parameterName=SizeY})
  MaterialExpressionScalarParameter_6 = ScalarParameter() {defaultValue=1, parameterName=RainAmount}
  MaterialExpressionScalarParameter_7 = ScalarParameter() {defaultValue=0.05, parameterName=Solubility}
  MaterialExpressionScalarParameter_8 = ScalarParameter() {defaultValue=0.05, parameterName=Evaporation}
  MaterialExpressionTextureCoordinate_0 = TextureCoordinate()
  Custom(
   HeightTex=@MaterialExpressionTextureObjectParameter_1
   TextureSize=@MaterialExpressionAppendVector_1
   RainAmount=@MaterialExpressionScalarParameter_6
   Solubility=@MaterialExpressionScalarParameter_7
   Evaporation=@MaterialExpressionScalarParameter_8
   UV=@MaterialExpressionTextureCoordinate_0
  ) {
   code="//HeightTex: input Texture Object containing heights.R and water.B\r\n//TextureSize: input for size\r\n//RainAmount: input for per texel rain contribution\r\n//Solubility: how much sediment the water can hold\r\n//Evaporation: how much water to evaporate each round\r\n\r\nfloat2 offsets[9] = {float2(0,0), float2(0,1),float2(0,2),float2(0,-1), float2(0,-2),float2(1,0),float2(2,0),float2(-1,0),float2(-2,0)};\r\n\r\nfloat heights[9] = {0,0,0,0,0,0,0,0,0};\r\nfloat water[9] =   {0,0,0,0,0,0,0,0,0};\r\nint orders[9] =    {0,0,0,0,0,0,0,0,0};\r\nint actives[9] =   {0,0,0,0,0,0,0,0,0};\r\n\r\nfloat2 texelsize = 1 / TextureSize;\r\nfloat watersum = 0;\r\n\r\nUV = (floor(UV * TextureSize)/TextureSize) + (texelsize * 0.5);\r\n\r\n//Sample Heightmap and water data\r\nfor (int i = 0; i < 9; i++)\r\n{\r\n\tfloat2 heightsample = HeightTex.SampleLevel(HeightTexSampler, UV + offsets[i] * texelsize, 0).rb;\r\n\theights[i] = heightsample.r;\r\n\tfloat watertemp = heightsample.g + RainAmount;\r\n\theights[i] -= RainAmount * Solubility; \r\n\r\n\twatersum += watertemp;\r\n\twater[i] = 0;\r\n}\r\n\r\n//sort heights by how many are lower than each\r\nfor (int j = 0; j < 9; j++)\r\n{\r\n\tfloat curheight = heights[j];\r\n\r\n\tfor (int k = 0; k < 9; k++)\r\n\t{\r\n\t\tfloat checkheight = heights[k];\r\n\t\tif (checkheight < curheight)\r\n\t\t{\r\n\t\t\torders[j] += 1;\r\n\t\t}\r\n\t}\r\n}\r\n\r\nfloat maxlevel = 9;\r\n\r\n//use sort data to iterate again\r\nfor (int l = 0; l < min(maxlevel,9); l++)\r\n{\r\n\r\n\tfloat curactivecells = 0;\r\n\tfloat curpooltop = 65000;\r\n\tfloat curpoolbottom = -65000;\r\n\t\r\n\tfor (int m = 0; m < 9; m++)\r\n\t{\r\n\t\tif(orders[m] <= l)\r\n\t\t{\r\n\t\t\tcuractivecells += 1;\r\n\t\t\tcurpoolbottom = max(curpoolbottom,heights[m]+water[m]);\r\n\t\t\tactives[m] = 1;\r\n\t\t}\r\n\t\telse\r\n\t\t{\r\n\t\t\tactives[m] = 0;\r\n\t\t}\r\n\t\tif(orders[m] > l )\r\n\t\t{\r\n\t\t\tcurpooltop = min(curpooltop,heights[m]);\r\n\t\t}\r\n\t\t\r\n\t}\r\n\r\n\tfloat curpoolsize = (curpooltop - curpoolbottom) * curactivecells;\r\n\tcurpoolsize = min(curpoolsize, watersum);\r\n\tcurpoolsize = max(0, curpoolsize);\r\n\r\n\tfor (int n = 0; n < 9; n++)\r\n\t{\r\n\t\tif(actives[n] == 1)\r\n\t\t{\r\n\t\t\twater[n] += curpoolsize / curactivecells;\r\n\t\t\twatersum -= curpoolsize / curactivecells;\r\n\t\t}\r\n\t}\r\n\t\r\n\r\n}\r\n\r\nheights[0] += water[0] * Evaporation * Solubility;\r\nwater[0] -= water[0] * Evaporation;\r\n\r\nreturn float3(heights[0],0,water[0]);"
   description=Erosion Pass
   inputs [
    {
     inputName=HeightTex
     input {expression=&self:MaterialExpressionTextureObjectParameter_1}
    }
    {inputName=TextureSize, input {expression=&self:MaterialExpressionAppendVector_1}}
    {inputName=RainAmount, input {expression=&self:MaterialExpressionScalarParameter_6}}
    {inputName=Solubility, input {expression=&self:MaterialExpressionScalarParameter_7}}
    {inputName=Evaporation, input {expression=&self:MaterialExpressionScalarParameter_8}}
    {inputName=UV, input {expression=&self:MaterialExpressionTextureCoordinate_0}}
   ]
  }
 ]
}
